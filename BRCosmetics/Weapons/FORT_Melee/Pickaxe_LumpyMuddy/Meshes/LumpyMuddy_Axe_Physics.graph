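# Physics asset for the LumpyMuddy pickaxe cosmetic mesh, simulating a pair of dangling bow tails on the weapon.
# The base pickaxe body carries two capsules, while each left and right tail is a two-bone chain with one capsule per bone.
# Angular-limited constraints hang each chain from the base, with wider primary swing and no twist at the root and slight twist at the outer segment.
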PhysicsAsset /BRCosmetics/Weapons/FORT_Melee/Pickaxe_LumpyMuddy/Meshes/LumpyMuddy_Axe_Physics
bodies {
 base_pickaxeBody [
  {shapeType=Capsule, length=4, radius=1.9, shapeName=base_pickaxeBody_capsule, center=<0.1076 -0.4921 0.1994>, rotation=<rot 90 0 0>}
  {shapeType=Capsule, length=28, radius=2.1, shapeName=base_pickaxeBody_capsule, center=<-7.4101 -2.175 0.0316>, rotation=<rot 90 0 0>}
 ]
 dyn_bowTail02_l [{shapeType=Capsule, length=2.5, radius=3, shapeName=dyn_Bow_l_capsule, center=<2.7309 1.1191 -1.0609>, rotation=<rot 69.9998 19.9998 -0.0001>}]
 dyn_bowTail02_r [{
   shapeType=Capsule
   length=3
   radius=2.5
   shapeName=dyn_Bow_l_capsule
   center=<2.9804 2.0212 0.525>
   rotation=<rot 80.0001 -159.9999 -179.9999>
  }]
 dyn_bowTail_l [{shapeType=Capsule, length=2.5, radius=1.5, shapeName=dyn_Bow_l_capsule, center=<3.5938 0 0.1783>, rotation=<rot 86.0841 0.0004 0.0004>}]
 dyn_bowTail_r [{shapeType=Capsule, length=2, radius=2, shapeName=dyn_Bow_l_capsule, center=<3.9342 0 0>, rotation=<rot 90 0 0>}]
}
constraints [
 {bone1Name=dyn_bowTail_l, bone2Name=base_pickaxeBody, swing1LimitDegrees=30, swing2LimitDegrees=10, twistLimitDegrees=0}
 {bone1Name=dyn_bowTail02_l, bone2Name=dyn_bowTail_l, swing1LimitDegrees=20, swing2LimitDegrees=20, twistLimitDegrees=2}
 {bone1Name=dyn_bowTail_r, bone2Name=base_pickaxeBody, swing1LimitDegrees=30, swing2LimitDegrees=10, twistLimitDegrees=0}
 {bone1Name=dyn_bowTail02_r, bone2Name=dyn_bowTail_r, swing1LimitDegrees=20, swing2LimitDegrees=20, twistLimitDegrees=2}
]
