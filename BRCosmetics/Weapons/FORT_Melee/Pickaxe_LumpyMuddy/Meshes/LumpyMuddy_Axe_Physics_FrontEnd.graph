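# Physics asset for the LumpyMuddy pickaxe's front-end variant, simulating the dangling bow and its tails on the weapon.
# Capsule bodies cover the pickaxe base and four two-bone dyn chains: left and right bow loops and left and right bow tails.
# Each chain's first bone is constrained to the base and its second bone to the first, with tight twist and moderate swing limits.
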PhysicsAsset /BRCosmetics/Weapons/FORT_Melee/Pickaxe_LumpyMuddy/Meshes/LumpyMuddy_Axe_Physics_FrontEnd
bodies {
 base_pickaxeBody [
  {shapeType=Capsule, length=4, radius=1.9, shapeName=base_pickaxeBody_capsule, center=<0.1076 -0.4921 0.1994>, rotation=<rot 90 0 0>}
  {shapeType=Capsule, length=28, radius=2.1, shapeName=base_pickaxeBody_capsule, center=<-7.4101 -2.175 0.0316>, rotation=<rot 90 0 0>}
 ]
 dyn_Bow02_l [{shapeType=Capsule, length=3, radius=3, shapeName=dyn_Bow_l_capsule, center=<2.7765 -0.1064 -2.6098>, rotation=<rot 69.9997 0.0001 0.0001>}]
 dyn_Bow02_r [{shapeType=Capsule, length=4.5, radius=3, shapeName=dyn_Bow_l_capsule, center=<2.3165 1.4418 3.4893>, rotation=<rot -63.196 -20.4272 27.6497>}]
 dyn_Bow_l [{shapeType=Capsule, length=1.5, radius=2, shapeName=dyn_Bow_l_capsule, center=<2.5 0 -0.2259>, rotation=<rot 90 0 0>}]
 dyn_Bow_r [{shapeType=Capsule, length=1.5, radius=2, shapeName=dyn_Bow_l_capsule, center=<2.5 0 0>, rotation=<rot 90 0 0>}]
 dyn_bowTail02_l [
  {shapeType=Capsule, length=4.3, radius=2, shapeName=dyn_Bow_l_capsule, center=<2.024 1.3266 -2.0536>, rotation=<rot 64.9978 19.9991 -0.0009>}
  {
   shapeType=Capsule
   length=4.3
   radius=2.3
   shapeName=dyn_bowTail02_l_capsule
   center=<-0.17 0.2957 0.2435>
   rotation=<rot 84.0257 -160.0033 -174.9996>
  }
 ]
 dyn_bowTail02_r [{shapeType=Capsule, length=4, radius=2.5, shapeName=dyn_Bow_l_capsule, center=<0.8971 1.295 -0.4564>, rotation=<rot 80.0372 14.9802 0.0186>}]
 dyn_bowTail_l [{shapeType=Capsule, length=2.5, radius=1.5, shapeName=dyn_Bow_l_capsule, center=<3.6123 0 0.4494>, rotation=<rot 88.9163 179.9946 179.9946>}]
 dyn_bowTail_r [{shapeType=Capsule, length=2, radius=2, shapeName=dyn_Bow_l_capsule, center=<3.3904 0.1751 -0.3739>, rotation=<rot 89.1288 -82.4153 -97.4153>}]
}
constraints [
 {bone1Name=dyn_Bow_l, bone2Name=base_pickaxeBody, swing1LimitDegrees=30, swing2LimitDegrees=22, twistLimitDegrees=2}
 {bone1Name=dyn_Bow02_l, bone2Name=dyn_Bow_l, swing1LimitDegrees=12, swing2LimitDegrees=22, twistLimitDegrees=2}
 {bone1Name=dyn_Bow_r, bone2Name=base_pickaxeBody, swing1LimitDegrees=30, swing2LimitDegrees=20, twistLimitDegrees=2}
 {bone1Name=dyn_Bow02_r, bone2Name=dyn_Bow_r, swing1LimitDegrees=12, swing2LimitDegrees=25, twistLimitDegrees=2}
 {bone1Name=dyn_bowTail_l, bone2Name=base_pickaxeBody, swing1LimitDegrees=30, swing2LimitDegrees=20, twistLimitDegrees=2}
 {bone1Name=dyn_bowTail02_l, bone2Name=dyn_bowTail_l, swing1LimitDegrees=20, swing2LimitDegrees=20, twistLimitDegrees=5}
 {bone1Name=dyn_bowTail_r, bone2Name=base_pickaxeBody, swing1LimitDegrees=30, swing2LimitDegrees=20, twistLimitDegrees=2}
 {bone1Name=dyn_bowTail02_r, bone2Name=dyn_bowTail_r, swing1LimitDegrees=15, swing2LimitDegrees=15, twistLimitDegrees=5}
]
