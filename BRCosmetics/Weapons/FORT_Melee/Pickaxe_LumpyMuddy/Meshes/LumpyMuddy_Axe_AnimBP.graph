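# Post-process style AnimBP for the LumpyMuddy pickaxe that adds secondary physics motion to its bows over a cached input pose.
# In gameplay, chained spring-driven AnimDynamics on left and right bow bones feed a RigidBody node; the front end blends to a separate RigidBody with its own physics asset.
# Each update, a nativized state function computes the gravity override for the bow AnimDynamics.
# The bow-tail AnimDynamics nodes are disconnected from the output.
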AnimBlueprint /BRCosmetics/Weapons/FORT_Melee/Pickaxe_LumpyMuddy/Meshes/LumpyMuddy_Axe_AnimBP
target_skeleton=&/LumpyMuddy_Axe_Skeleton
graphs [
 AnimGraph {nodes [
   n1 = Anim Dynamics - Bone: dyn_bowTail_r!(ComponentPose, Alpha=1, GravityOverride:Vector="0,0,0", ExternalForce="0,0,0") -> (Pose) {
    angularDampingOverride=1
    angularSpringConstant=6000
    bAngularSpring=true
    bOverrideAngularDamping=true
    bUseGravityOverride=true
    gravityScale=0.35
    alphaBoolBlend {blendOption=Linear}
    boundBone {boneName=dyn_bowTail_r}
    componentAppliedLinearAccClamp=<5000 3000 3000>
    componentLinearAccScale=<0.5 0.3 0.3>
    componentLinearVelScale=<0.6 0.4 0.4>
    gravityOverride {z=-980}
    physicsBodyDefinitions [{
      boundBone {boneName=dyn_bowTail_r}
      boxExtents=<18 18 18>
      constraintSetup {angularLimitsMax=<30 80 3>, angularLimitsMin=<-30 -8 -30>, angularTarget=<1 0 0>}
      localJointOffset=<0.5 0.5 0>
     }]
   }
   Anim Dynamics - Bone: dyn_bowTail02_r!(ComponentPose=@n1, Alpha=1, GravityOverride:Vector="0,0,0", ExternalForce="0,0,0") -> (Pose) {
    angularDampingOverride=1
    angularSpringConstant=12000
    bAngularSpring=true
    bOverrideAngularDamping=true
    bUseGravityOverride=true
    gravityScale=0.35
    alphaBoolBlend {blendOption=Linear}
    boundBone {boneName=dyn_bowTail02_r}
    componentAppliedLinearAccClamp=<14000 3000 3000>
    componentLinearAccScale=<0.7 0.3 0.3>
    componentLinearVelScale=<0.8 0.4 0.4>
    gravityOverride {z=-980}
    physicsBodyDefinitions [{
      boundBone {boneName=dyn_bowTail02_r}
      boxExtents=<18 18 50>
      constraintSetup {angularLimitsMax=<25 20 45>, angularLimitsMin=<-25 -20 -45>, angularTarget=<1 0 0>}
      localJointOffset=<3 0 0>
     }]
   }
   n2 = Anim Dynamics - Bone: dyn_bowTail_l!(ComponentPose, Alpha=1, GravityOverride:Vector="0,0,0", ExternalForce="0,0,0") -> (Pose) {
    angularDampingOverride=1
    angularSpringConstant=6000
    bAngularSpring=true
    bOverrideAngularDamping=true
    bUseGravityOverride=true
    gravityScale=0.35
    alphaBoolBlend {blendOption=Linear}
    boundBone {boneName=dyn_bowTail_l}
    componentAppliedLinearAccClamp=<5000 3000 3000>
    componentLinearAccScale=<0.5 0.3 0.3>
    componentLinearVelScale=<0.6 0.4 0.4>
    gravityOverride {z=-980}
    physicsBodyDefinitions [{
      boundBone {boneName=dyn_bowTail_l}
      boxExtents=<18 18 18>
      constraintSetup {angularLimitsMax=<40 8 3>, angularLimitsMin=<-40 -80 -40>, angularTarget=<1 0 0>}
      localJointOffset=<0.7 0.7 0>
     }]
   }
   Anim Dynamics - Bone: dyn_bowTail02_l!(ComponentPose=@n2, Alpha=1, GravityOverride:Vector="0,0,0", ExternalForce="0,0,0") -> (Pose) {
    angularDampingOverride=1
    angularSpringConstant=4000
    bAngularSpring=true
    bOverrideAngularDamping=true
    bUseGravityOverride=true
    gravityScale=0.35
    alphaBoolBlend {blendOption=Linear}
    boundBone {boneName=dyn_bowTail02_l}
    componentAppliedLinearAccClamp=<14000 3000 3000>
    componentLinearAccScale=<0.8 0.3 0.3>
    componentLinearVelScale=<0.9 0.4 0.4>
    gravityOverride {z=-980}
    physicsBodyDefinitions [{
      boundBone {boneName=dyn_bowTail02_l}
      boxExtents=<18 18 50>
      constraintSetup {angularLimitsMax=<80 8 30>, angularLimitsMin=<-80 -80 -30>, angularTarget=<1 0 0>}
      localJointOffset=<2 0 0>
     }]
   }
   n3 = Get Gravity Override()
   n4 = Anim Dynamics - Bone: dyn_Bow_r!(
    ComponentPose=Local To Component(LocalPose=Use cached pose 'CRCache_HydraTrumpet'())
    Alpha=1
    GravityOverride:Vector=@n3
    ExternalForce="0,0,0"
   ) -> (Pose) {
    angularDampingOverride=0.99
    angularSpringConstant=15000
    bAngularSpring=true
    bOverrideAngularDamping=true
    bUseGravityOverride=true
    gravityScale=0.35
    alphaBoolBlend {blendOption=Linear}
    boundBone {boneName=dyn_Bow_r}
    componentAppliedLinearAccClamp=<5000 3000 3000>
    componentLinearAccScale=<0.7 0.3 0.3>
    componentLinearVelScale=<0.6 0.4 0.4>
    gravityOverride {z=-980}
    physicsBodyDefinitions [{
      boundBone {boneName=dyn_Bow_r}
      boxExtents=<30 30 30>
      constraintSetup {angularLimitsMax=<50 80 30>, angularLimitsMin=<-50 0 -30>, angularTarget=<1 0 0>}
      localJointOffset=<0.7 0 0>
     }]
   }
   n5 = Anim Dynamics - Bone: dyn_Bow02_r!(ComponentPose=@n4, Alpha=1, GravityOverride:Vector=@n3, ExternalForce="0,0,0") -> (Pose) {
    angularDampingOverride=0.99
    angularSpringConstant=3500
    bAngularSpring=true
    bOverrideAngularDamping=true
    bUseGravityOverride=true
    gravityScale=0.35
    alphaBoolBlend {blendOption=Linear}
    boundBone {boneName=dyn_Bow02_r}
    componentAppliedLinearAccClamp=<14000 3000 3000>
    componentLinearAccScale=<0.7 0.3 0.3>
    componentLinearVelScale=<0.8 0.4 0.4>
    gravityOverride {z=-980}
    physicsBodyDefinitions [{
      boundBone {boneName=dyn_Bow02_r}
      boxExtents=<30 30 30>
      constraintSetup {angularLimitsMax=<20 20 70>, angularLimitsMin=<-20 -10 -35>, angularTarget=<1 0 0>}
      localJointOffset=<1.2 0 0>
     }]
   }
   n6 = Get Gravity Override()
   n7 = Anim Dynamics - Bone: dyn_Bow_l!(ComponentPose=@n5, Alpha=1, GravityOverride:Vector=@n6, ExternalForce="0,0,0") -> (Pose) {
    angularDampingOverride=0.99
    angularSpringConstant=17000
    bAngularSpring=true
    bOverrideAngularDamping=true
    bUseGravityOverride=true
    gravityScale=0.35
    alphaBoolBlend {blendOption=Linear}
    boundBone {boneName=dyn_Bow_l}
    componentAppliedLinearAccClamp=<5000 3000 3000>
    componentLinearAccScale=<0.7 0.3 0.3>
    componentLinearVelScale=<0.8 0.4 0.4>
    gravityOverride {z=-980}
    physicsBodyDefinitions [{
      boundBone {boneName=dyn_Bow_l}
      boxExtents=<30 30 30>
      constraintSetup {angularLimitsMax=<50 8 70>, angularLimitsMin=<-50 -80 -70>, angularTarget=<1 0 0>}
      localJointOffset=<0.7 0 0>
     }]
   }
   n8 = Anim Dynamics - Bone: dyn_Bow02_l!(ComponentPose=@n7, Alpha=1, GravityOverride:Vector=@n6, ExternalForce="0,0,0") -> (Pose) {
    angularDampingOverride=0.99
    angularSpringConstant=5000
    bAngularSpring=true
    bOverrideAngularDamping=true
    bUseGravityOverride=true
    gravityScale=0.35
    alphaBoolBlend {blendOption=Linear}
    boundBone {boneName=dyn_Bow02_l}
    componentAppliedLinearAccClamp=<14000 3000 3000>
    componentLinearAccScale=<0.7 0.3 0.3>
    componentLinearVelScale=<0.8 0.4 0.4>
    gravityOverride {z=-980}
    physicsBodyDefinitions [{
      boundBone {boneName=dyn_Bow02_l}
      boxExtents=<30 30 30>
      constraintSetup {angularLimitsMax=<20 8 80>, angularLimitsMin=<-20 -20 -50>, angularTarget=<1 0 0>}
      localJointOffset=<1.3 0 0>
     }]
   }
   n9 = RigidBody(ComponentPose=@n8) {
    bClampLinearTranslationLimitToRefPose=true
    bOverrideWorldGravity=true
    overridePhysicsAsset=&/LumpyMuddy_Axe_Physics
    simulationSpace=BaseBoneSpace
    alphaBoolBlend {blendOption=Linear}
    baseBoneRef {boneName=b_AR_Root}
    componentAppliedLinearAccClamp=<3000 3000 3000>
    overrideWorldGravity {z=-500}
    simSpaceSettings {
     dampingAlpha=0
     maxAngularAcceleration=3
     maxAngularVelocity=1
     maxLinearAcceleration=250
     maxLinearVelocity=150
     worldAlpha=1
     externalLinearDragV=<0.2 3 5>
    }
   }
   n10 = RigidBody(ComponentPose=Local To Component(LocalPose=Use cached pose 'CRCache_HydraTrumpet'())) {
    bOverrideWorldGravity=true
    overridePhysicsAsset=&/LumpyMuddy_Axe_Physics_FrontEnd
    simulationSpace=BaseBoneSpace
    alphaBoolBlend {blendOption=Linear}
    baseBoneRef {boneName=b_AR_Root}
    componentAppliedLinearAccClamp=<15000 15000 15000>
    overrideWorldGravity {z=-980}
    simSpaceSettings {
     dampingAlpha=0.5
     maxAngularAcceleration=600
     maxAngularVelocity=150
     maxLinearAcceleration=1200
     maxLinearVelocity=400
     worldAlpha=1
     externalLinearDragV=<1 1 1>
    }
   }
   Output Pose(Result=Blend Poses by bool(BlendPose_0=Component To Local(ComponentPose=@n10), BlendPose_1=Component To Local(ComponentPose=@n9), bActiveValue=Get bIsInFrontEnd()) {blendType=Linear}) {layerGroup=None}
   AnimGraphNode_SubInput_0 = Input Pose()
   Save cached pose 'CRCache_HydraTrumpet'(Pose=@AnimGraphNode_SubInput_0) {cacheName=CRCache_HydraTrumpet}
  ]}
 EventGraph {nodes [
   Event BlueprintUpdateAnimation()
   > n1 = "Gravity Override State Function (Parameter Asset) (Nativized)"(
    CharacterPartAnimInstance=Self-Reference()
    GameTimeInSeconds=GetGameTimeInSeconds()
    DebugJointName=spine_05
    Parameters=&/LumpyMuddy_Axe_Parameters
   )
   > Set Gravity Override(Gravity Override=@n1)
   TryGetPawnOwner()
  ]}
]
variables [Gravity Override {type=Vector}]
